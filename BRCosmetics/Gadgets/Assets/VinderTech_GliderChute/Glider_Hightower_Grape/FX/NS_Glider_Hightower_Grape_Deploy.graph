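# Niagara one-shot burst effect for deploying the Hightower Grape glider, built from five instantaneous-burst emitters.
# A scaling wind mesh, a shockwave mesh with animated material parameters and short-lived sprite spikes form the initial pop.
# Leaf sprites and tumbling twig meshes inherit velocity and fly outward from spheres with downward push and drag.
# Leaves live longest, combining upward per-particle gravity, a custom orbit module and curve-decayed spin to flutter.
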NiagaraSystem /BRCosmetics/Gadgets/Assets/VinderTech_GliderChute/Glider_Hightower_Grape/FX/NS_Glider_Hightower_Grape_Deploy
dependencies {
 usedDataInterfaces [
  &NiagaraDataInterfaceCurve
  &NiagaraDataInterfaceVector2DCurve
  &NiagaraDataInterfaceVectorCurve
 ]
 usedDynamicInputs []
 usedModules [
  &self:SystemSpawnScript.SystemScriptSource.SystemScriptGraph.NiagaraNodeAssignment_0.SetVariables_#1
  &SystemState
 ]
}
emitters {
 DirectionalBurstSpikes {
  emitterUpdateScript [
   &EmitterState()
   &SpawnBurst_Instantaneous(Spawn Count=&UniformRangedInt(Minimum=18, Maximum=24), Use Spawn Probability=false)
  ]
  particleSpawnScript [
   /Niagara/Modules/Spawn/Initialization/InitializeParticle(
    Color=&UniformRangedLinearColor(Minimum=<rgba 1 1 1 0.4>, Maximum=<rgba 1 1 1 0.25>)
    Lifetime=&UniformRangedFloat(Minimum=0.15, Maximum=0.25)
    Mass=&UniformRangedFloat(Minimum=0.75, Maximum=1.75)
    Module.Write Color=true
    Module.Write Lifetime=true
    Module.Write Mass=true
    Module.Write Material Random=true
    Module.Write Position=true
    Module.Write Scale=false
    Module.Write SpriteRotation=false
    Module.Write SpriteSize=true
    Position=&SimulationPosition()
    Sprite Size=&UniformRangedVector2D(Minimum=<10 80>, Maximum=<50 150>)
   )
   /Niagara/Modules/Spawn/Location/SphereLocation(MaskSpawnGroup=false, Module.Override Randomness=false, Sphere Radius=60, Surface Only=false)
   &AddVelocityFromPoint(Module.Use Falloff Distance=false, Velocity Strength=500)
   &DynamicMaterialParameters(
    Index 0 Param 1=50
    Module.Param0WriteEnabled=true
    Module.Param1WriteEnabled=true
    Module.Param2WriteEnabled=true
    Module.Param3WriteEnabled=true
   )
  ]
  particleUpdateScript [
   &ParticleState()
   &GravityForce() {enabled=false}
   &Drag(Use Linear Drag=true, Use Rotational Drag=false) {enabled=false}
   &ScaleColor(
    Scale Alpha=&FloatFromCurve(FloatCurve=DataInterface() {propertyValues {Curve {keys [
         {interpMode=RCIM_Cubic, leaveTangent=-0.0594, tangentMode=RCTM_Break, time=-0.0029, value=1.0303}
         {arriveTangent=-0.936, interpMode=RCIM_Cubic, leaveTangent=-0.936, tangentMode=RCTM_User, time=0.0887, value=0.3527}
         <at 1.0011 0 RCIM_Cubic>
        ]}}})
    ScaleA=true
    ScaleRGB=true
   )
   &SolveForcesAndVelocity(Clamp Velocity=false, Limit Acceleration=false)
   &ScaleSpriteSize(Scale Factor=&Vector2DFromCurve(
     Vector2Curve=DataInterface() {propertyValues {
       XCurve {keys [<at 0.0014 0.4615>,<at 1 1>]}
       YCurve {keys [{arriveTangent=1.6918, interpMode=RCIM_Cubic, leaveTangent=1.6918, tangentMode=RCTM_User, time=-0.0006, value=0.4523}, <at 1 1>]}
      }}
     ScaleCurve=<1 1>
    ))
  ]
  rendererClasses [&NiagaraSpriteRendererProperties]
  renderers [{rendererClass=&NiagaraSpriteRendererProperties}]
 }
 HT_LoopShockwave_CamF {
  emitterUpdateScript [
   &EmitterState()
   &SpawnBurst_Instantaneous(Spawn Count=1, Use Spawn Probability=false)
  ]
  particleSpawnScript [
   &InitializeParticle(
    Color=<rgba 3 3 3 0.3>
    Lifetime=0.4
    Mesh Scale=&VectorFromFloat(Value=3)
    Module.Write Color=true
    Module.Write Lifetime=true
    Module.Write Mass=true
    Module.Write Material Random=true
    Module.Write Position=false
    Module.Write Scale=true
    Module.Write SpriteRotation=false
    Module.Write SpriteSize=false
    Position=&SimulationPosition()
   )
   /Niagara/Modules/Spawn/Orientation/V2/InitialMeshOrientation(
    Orientation Axis:Vector3f=<0 0 1>
    Orientation Vector=<0 1 0>
    Rotation=&MakeVector(X=&UniformRangedFloat(Minimum=0, Maximum=1), Y=0, Z=0)
    Use Orientation Vector=true
    Use Rotation Vector=true
   )
  ]
  particleUpdateScript [
   &ParticleState()
   &ScaleMeshSize(Scale Factor=/Niagara/DynamicInputs/ScaleByCurve/ScaleVectorByCurve(
     InitialValue:Vector3f=<1 1 1>
     ScaleCurve:NiagaraDataInterfaceVectorCurve=DataInterface() {propertyValues {
       XCurve {keys [
         {arriveTangent=7.5627, interpMode=RCIM_Cubic, leaveTangent=3.2485, tangentMode=RCTM_User, time=0, value=0}
         {interpMode=RCIM_Cubic, tangentMode=RCTM_User, time=1, value=1}
        ]}
       YCurve {keys [
         {arriveTangent=9.2714, interpMode=RCIM_Cubic, leaveTangent=3.2485, tangentMode=RCTM_User, time=0, value=0}
         {interpMode=RCIM_Cubic, tangentMode=RCTM_User, time=1, value=1}
        ]}
       ZCurve {keys [
         {arriveTangent=8.0047, interpMode=RCIM_Cubic, leaveTangent=3.2485, tangentMode=RCTM_User, time=0, value=0}
         {interpMode=RCIM_Cubic, tangentMode=RCTM_User, time=1, value=1}
        ]}
      }}
     CurveIndex:NiagaraFloat=$Particles.NormalizedAge
    ))
   &ScaleColor(
    Scale Alpha=&FloatFromCurve(FloatCurve=DataInterface() {propertyValues {Curve {keys [
         {interpMode=RCIM_Cubic, tangentMode=RCTM_User, time=0.0023, value=0}
         {arriveTangent=2.0824, interpMode=RCIM_Cubic, leaveTangent=0.0112, tangentMode=RCTM_Break, time=0.0552, value=0.9565}
         {arriveTangent=-0.1758, interpMode=RCIM_Cubic, leaveTangent=-0.1758, tangentMode=RCTM_User, time=1.0011, value=0.002}
        ]}}})
    ScaleA=true
    ScaleRGB=true
   )
   /SetVariables(Particles.LightColor=&MakeLinearColorFromVectorAndFloat(
     "Vector (RGB)"=&VectorFromCurve(VectorCurve=DataInterface() {propertyValues {
        XCurve {keys [
          {interpMode=RCIM_Cubic, tangentMode=RCTM_Break, time=-0.0014, value=9.2729}
          {arriveTangent=-15.7559, interpMode=RCIM_Cubic, leaveTangent=-0.3065, tangentMode=RCTM_Break, time=0.3017, value=-0.1656}
          <at 1 0>
         ]}
        YCurve {keys [
          {interpMode=RCIM_Cubic, leaveTangent=30.1692, tangentMode=RCTM_Break, time=0, value=0}
          {interpMode=RCIM_Cubic, leaveTangent=-34.8929, tangentMode=RCTM_Break, time=0.2237, value=12}
          {interpMode=RCIM_Cubic, tangentMode=RCTM_Break, time=1, value=0}
         ]}
        ZCurve {keys [<at 0 0>,<at 0.0627 0>,<at 1 0>]}
       }})
     "Float (Alpha)"=1
    )) {isSetParameters=true}
   /SetVariables(Particles.LightRadius=&FloatFromCurve(FloatCurve=DataInterface() {propertyValues {Curve {keys [<at 0.0021 0.0215>,<at 0.1455 0.1865>,<at 0.9014 1>]}}}, Scale Curve=125)) {isSetParameters=true}
   &DynamicMaterialParameters(
    Index 0 Param 1=&FloatFromCurve(FloatCurve=DataInterface() {propertyValues {Curve {keys [<at 0 2.5>,<at 1 0>]}}})
    Index 0 Param 2=&FloatFromCurve(FloatCurve=DataInterface() {propertyValues {Curve {keys [<at 0 3>,<at 1 0>]}}}, Scale Curve=5)
    Index 0 Param 3=1
    Module.Param0WriteEnabled=true
    Module.Param1WriteEnabled=true
    Module.Param2WriteEnabled=true
    Module.Param3WriteEnabled=true
   )
  ]
  rendererClasses [&NiagaraMeshRendererProperties]
  renderers [{rendererClass=&NiagaraMeshRendererProperties, rendererIndex=0}]
 }
 Leaves {
  emitterUpdateScript [
   &EmitterState()
   &SpawnBurst_Instantaneous(Spawn Count=&UniformRangedInt(Minimum=20, Maximum=25), Use Spawn Probability=false) {module=SpawnBurst_Instantaneous001}
  ]
  particleSpawnScript [
   &InitializeParticle(
    Lifetime=&UniformRangedFloat(Minimum=2, Maximum=3)
    Module.Write Color=true
    Module.Write Lifetime=true
    Module.Write Mass=true
    Module.Write Material Random=true
    Module.Write Position=false
    Module.Write Scale=false
    Module.Write SpriteRotation=true
    Module.Write SpriteSize=true
    Position=&SimulationPosition()
    Sprite Rotation=&UniformRangedFloat(Minimum=0, Maximum=360)
    Sprite Size=&UniformRangedVector2D(Minimum=<6 15>, Maximum=<10 20>)
   )
   &InheritVelocity(Module.Limit Speed=false, Velocity Limit=1000)
   &SphereLocation(
    Hemisphere Z=true
    MaskSpawnGroup=false
    Module.Override Randomness=false
    Non Uniform Scale=<1 1 -1>
    Sphere Origin:NiagaraPosition=/Niagara/DynamicInputs/Helpers/SimulationPosition()
    Sphere Radius=30
    Surface Only=false
   )
   &AddVelocity(
    Coordinate Space=::World
    Velocity=&Multiply_Vector(
     A=$Output.SphereLocation.SphereNormal
     B=&VectorFromFloat(Value=&UniformRangedFloat(Minimum=1000, Maximum=1750))
    )
   ) {module=AddVelocity001}
   &AddVelocity(Coordinate Space=::World, Velocity=&UniformRangedVector(Minimum=<0 0 -100>, Maximum=<0 0 -400>)) {module=AddVelocity002}
   /SetVariables(Particles.Gravity-InitialRange:Vector3f=&UniformRangedVector(Minimum=<0 0 1000>, Maximum=<0 0 1600>)) {isSetParameters=true}
  ]
  particleUpdateScript [
   &ParticleState()
   &GravityForce(Gravity=Linked() {linkedVariable {name=Particles.Gravity-InitialRange, type=Vector3f}})
   &Drag(
    Drag=&FloatFromCurve(FloatCurve=DataInterface() {propertyValues {Curve {keys [<at 0 5>,<at 1 1>]}}})
    Use Linear Drag=true
    Use Rotational Drag=false
   )
   /Game/Effects/Niagara/Modules/Orbit/FN_Orbit_v3(
    Delta:NiagaraFloat=&Add_Float(A=$Emitter.Age, B=&MakeFloatFromInt(Int=/Niagara/DynamicInputs/Execution/ReturnParticleID()))
    Height min/max:Vector2f=<12 12>
    Radius min/max:Vector2f=<12 12>
    Rotation Offset min/max:Vector2f=<5 1>
    Rotation Rate min/max:Vector2f=<0.5 2>
    To Axis:Vector3f=&Lerp_Vector(A=<-1 -1 -1>, B=<1 1 1>, Alpha=$Particles.Initial.MaterialRandom)
   )
   &SolveForcesAndVelocity(Clamp Velocity=false, Limit Acceleration=false)
   &SpriteRotationRate(Rotation Rate=&Lerp_Float(A=-1, B=1, Alpha=$Particles.Initial.MaterialRandom))
   &ScaleSpriteSize(Scale Factor=&Vector2DFromFloat(Value=&FloatFromCurve(FloatCurve=DataInterface() {propertyValues {Curve {keys [
          {interpMode=RCIM_Cubic, tangentMode=RCTM_Break, time=0, value=1.5}
          {arriveTangent=0.002, interpMode=RCIM_Cubic, leaveTangent=0.002, tangentMode=RCTM_Break, time=0.0652, value=1}
          {arriveTangent=-0.0496, interpMode=RCIM_Cubic, leaveTangent=-1.2885, tangentMode=RCTM_Break, time=0.7985, value=1}
          {interpMode=RCIM_Cubic, tangentMode=RCTM_Break, time=1.0011, value=0.002}
         ]}}})))
   &SpriteRotationRate(Rotation Rate=&Multiply_Float(
     A=&Lerp_Float(A=-800, B=800, Alpha=$Particles.Initial.MaterialRandom)
     B=&FloatFromCurve(FloatCurve=DataInterface() {propertyValues {Curve {keys [<at 0 8.9533>,<at 0.0642 1.9408>,<at 0.1376 0.9783>,<at 1 0.6849>]}}})
    )) {module=SpriteRotationRate001}
  ]
  rendererClasses [&NiagaraSpriteRendererProperties]
  renderers [{rendererClass=&NiagaraSpriteRendererProperties}]
 }
 Twigs {
  emitterUpdateScript [
   &EmitterState()
   &SpawnBurst_Instantaneous(Spawn Count=15, Use Spawn Probability=false)
  ]
  particleSpawnScript [
   &InitializeParticle(
    Lifetime=&UniformRangedFloat(Minimum=0.8, Maximum=1)
    Mesh Scale=&Multiply_Vector(
     A=<1 1.2 1.2>
     B=&VectorFromFloat(Value=&UniformRangedFloat(Minimum=0.35, Maximum=1))
    )
    Module.Write Color=true
    Module.Write Lifetime=true
    Module.Write Mass=true
    Module.Write Material Random=true
    Module.Write Position=false
    Module.Write Scale=true
    Module.Write SpriteRotation=false
    Module.Write SpriteSize=false
    Position=/Niagara/Functions/Localspace/SimulationPosition()
    Sprite Rotation=&UniformRangedFloat(Minimum=0, Maximum=360)
    Sprite Size=&UniformRangedVector2D(Minimum=<4 10>, Maximum=<7 12>)
   )
   &InheritVelocity(Module.Limit Speed=false)
   &SphereLocation(
    MaskSpawnGroup=false
    Module.Override Randomness=false
    Sphere Origin:NiagaraPosition=/Niagara/DynamicInputs/Helpers/SimulationPosition()
    Sphere Radius=30
    Surface Only=false
   )
   &AddVelocity(Velocity=&Multiply_Vector(
     A=$Output.SphereLocation.SphereNormal
     B=&VectorFromFloat(Value=&UniformRangedFloat(Minimum=650, Maximum=1250))
    )) {module=AddVelocity001}
   &AddVelocity(Coordinate Space=::World, Velocity=&UniformRangedVector(Minimum=<0 0 -100>, Maximum=<0 0 -300>))
   &InitialMeshOrientation(
    Rotation=&UniformRangedVector(Minimum=<-1 -1 -1>, Maximum=<1 1 1>)
    Use Orientation Vector=true
    Use Rotation Vector=true
   )
  ]
  particleUpdateScript [
   &ParticleState()
   &Drag(Use Linear Drag=true, Use Rotational Drag=false)
   /Niagara/Modules/Update/Orientation/V2/UpdateMeshOrientation(
    Rotation Rate:NiagaraFloat=&Lerp_Float(A=-2.5, B=2.5, Alpha=$Particles.Initial.MaterialRandom)
    Rotation Vector:Vector3f=&UniformRangedVector(Minimum=<0 0 0>, Maximum=<1 1 1>)
   )
   &SolveForcesAndVelocity(Clamp Velocity=false, Limit Acceleration=false)
   &ScaleMeshSize(Scale Factor=&VectorFromFloat(Value=&FloatFromCurve(FloatCurve=DataInterface() {propertyValues {Curve {keys [<at 0 1>,<at 0.6963 0.9473>,<at 1 0>]}}})))
  ]
  rendererClasses [&NiagaraMeshRendererProperties]
  renderers [{rendererClass=&NiagaraMeshRendererProperties, rendererIndex=0}]
 }
 WindMeshBurt {
  emitterUpdateScript [
   &EmitterState()
   &SpawnBurst_Instantaneous(Spawn Count=1, Use Spawn Probability=false)
  ]
  particleSpawnScript [&InitializeParticle(
    Color=<rgba 1 1 1 0.75>
    Lifetime=0.225
    Mesh Scale=&VectorFromFloat(Value=2)
    Module.Write Color=true
    Module.Write Lifetime=true
    Module.Write Mass=true
    Module.Write Material Random=true
    Module.Write Position=true
    Module.Write Scale=true
    Module.Write SpriteRotation=false
    Module.Write SpriteSize=false
    Position=/Niagara/Functions/Localspace/SimulationPosition()
    Sprite Size=<25 25>
   )]
  particleUpdateScript [
   &ParticleState()
   &ScaleColor(
    Scale Alpha=&FloatFromCurve(FloatCurve=DataInterface() {propertyValues {Curve {keys [<at 0.2946 1>,<at 1 0>]}}})
    Scale RGB=&MakeVectorFromLinearColorRGB(Color=<rgba 1 1 1 1>)
    ScaleA=true
    ScaleRGB=true
   )
   &ScaleMeshSize(Scale Factor=&VectorFromFloat(Value=&FloatFromCurve(FloatCurve=DataInterface() {propertyValues {Curve {keys [{arriveTangent=2.5191, interpMode=RCIM_Cubic, leaveTangent=2.5191, tangentMode=RCTM_User, time=0, value=0}, <at 1 1>]}}})))
  ]
  rendererClasses [&NiagaraMeshRendererProperties]
  renderers [{rendererClass=&NiagaraMeshRendererProperties, rendererIndex=0}]
 }
}
summary {
 systemName=NS_Glider_Hightower_Grape_Deploy
 emitters [
  {emitterName=WindMeshBurt, rendererClasses [&NiagaraMeshRendererProperties]}
  {emitterName=DirectionalBurstSpikes, rendererClasses [&NiagaraSpriteRendererProperties]}
  {emitterName=Leaves, rendererClasses [&NiagaraSpriteRendererProperties]}
  {emitterName=Twigs, rendererClasses [&NiagaraMeshRendererProperties]}
  {emitterName=HT_LoopShockwave_CamF, rendererClasses [&NiagaraMeshRendererProperties]}
 ]
 userVariables [User.windowSize {type=Vector3f, description="", defaultValue {struct=Vector3f, value=<0 0 0>}}]
}
