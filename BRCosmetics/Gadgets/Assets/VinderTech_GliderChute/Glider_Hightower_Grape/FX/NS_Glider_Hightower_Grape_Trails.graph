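# Niagara trail effect for the Hightower Grape glider: two anim-trail ribbons plus shed leaf sprites and twig meshes.
# Ribbons follow bone or socket pairs, with spawning switched by bIsFullyDeployed.
# Leaves and twigs spawn intermittently along the skeletal mesh via partitioned BoneFlow, then tumble under drag and gravity.
# A system integrator value drives ribbon alpha, debris spawn probability and a launch velocity blend that flips direction.
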NiagaraSystem /BRCosmetics/Gadgets/Assets/VinderTech_GliderChute/Glider_Hightower_Grape/FX/NS_Glider_Hightower_Grape_Trails
dependencies {
 usedDataInterfaces [&NiagaraDataInterfaceCurve]
 usedDynamicInputs []
 usedModules [&SystemState, /Game/Effects/Niagara/Modules/Math/FN_Integrator/FN_Integrator_v2]
}
emitters {
 FN_AnimTrail_V4001 {
  emitterUpdateScript [
   &EmitterState()
   &FN_SpawnAnimTrail_v2(
    Number of Anim Trails=2
    Spawn Interval:NiagaraInt32=&MakeCustomIntFromBool(Bool=$User.bIsFullyDeployed, True INT=1, False INT=0)
   )
  ]
  particleSpawnScript [
   /Game/Effects/Niagara/Modules/AnimTrail/FN_InitializeAnimTrail/FN_InitializeAnimTrailBoneIndecies(First Bone or Socket Index=1)
   &FN_InitializeAnimTrailBoneIndecies(AnimTrail Selection=1, First Bone or Socket Index=3, Second Bone or Socket Index=2) {module=FN_InitializeAnimTrailBoneIndecies001}
   /Game/Effects/Niagara/Modules/AnimTrail/FN_InitializeAnimTrail/FN_InitializeAnimTrail(AnimTrailWidth=1, Color=<rgba 1.1 1.0106 0.8855 1>, Lifetime=0.175)
  ]
  particleUpdateScript [
   &InheritVelocity(Module.Limit Speed=true)
   &SolveForcesAndVelocity(Clamp Velocity=false, Limit Acceleration=false)
   /Game/Effects/Niagara/Modules/AnimTrail/FN_UpdateAnimTrail/FN_UpdateAnimTrail()
   &ScaleColor(
    Scale Alpha=&Multiply_Float(
     A=&FloatFromCurve(FloatCurve=DataInterface() {propertyValues {Curve {keys [<at 0 0>,<at 0.2538 0.7462>,<at 1 0>]}}})
     B=$System.FN_Integrator_v2.Attr
    )
    Scale RGB=&MakeVectorFromLinearColorRGB(Color=<rgba 1 0.8915 0.8 1>)
    ScaleA=true
    ScaleRGB=true
   )
   &ScaleRibbonWidth(Ribbon Width Scale=&FloatFromCurve(FloatCurve=DataInterface() {propertyValues {Curve {keys [<at 0 1>,<at 1 1>]}}}))
  ]
  rendererClasses [&NiagaraRibbonRendererProperties]
  renderers [{rendererClass=&NiagaraRibbonRendererProperties}]
 }
 Leaves001 {
  emitterUpdateScript [
   /SetVariables(Emitter.FullyDeployed:NiagaraFloat=&Lerp_Float(A=0.1, B=1, Alpha=$System.FN_Integrator_v2.Attr)) {isSetParameters=true, module=SetVariables_#1}
   &EmitterState(
    Life Cycle Mode=::NewEnumerator1
    Loop Duration=&UniformRangedFloat(Minimum=0.1, Maximum=2)
    Recalculate Duration Each Loop=true
    UseLoopDelay=false
   )
   /Game/Effects/Niagara/Modules/Partition/FN_PartitionAttributes/FN_PartitionAttributes_v2(Particles Per Partition:NiagaraInt32=&UniformRangedInt(Minimum=5, Maximum=20), Partitions=2)
   &SpawnRate(
    Spawn Probability=$Emitter.FullyDeployed
    SpawnRate=&FloatFromCurve(FloatCurve=DataInterface() {propertyValues {Curve {keys [<at 0.7 0>,<at 0.75 11>,<at 1 11>]}}}, CurveIndex=$Emitter.LoopedAge)
    Use Spawn Probability=true
   )
  ]
  particleSpawnScript [
   /Game/Effects/Niagara/Modules/Partition/FN_ParticlePartition_v2/FN_ParticlePartition_v2()
   /Niagara/Modules/Spawn/Initialization/InitializeParticle(
    Lifetime=&UniformRangedFloat(Minimum=2, Maximum=3)
    Module.Write Color=true
    Module.Write Lifetime=true
    Module.Write Mass=true
    Module.Write Material Random=true
    Module.Write Position=false
    Module.Write Scale=false
    Module.Write SpriteRotation=true
    Module.Write SpriteSize=true
    Position=&SimulationPosition()
    Sprite Rotation=&UniformRangedFloat(Minimum=0, Maximum=360)
    Sprite Size=&UniformRangedVector2D(Minimum=<6 9>, Maximum=<8 14>)
   )
   /Game/Effects/Niagara/Modules/BoneFlow/FN_Initialize_BoneFlow_Attributes_v2(Angle=1, Radius=100)
   /Game/Effects/Niagara/Modules/BoneFlow/FN_BoneFlow_v2/FN_BoneFlow_v2(
    Angle=-90
    Radius:NiagaraFloat=&UniformRangedFloat(Minimum=25, Maximum=90)
    SkeletalMesh:NiagaraDataInterfaceSkeletalMesh=$User.SkeletalMesh
    T:NiagaraFloat=$Particles.T
   )
   &AddVelocity(
    Scale Added Velocity=&Lerp_Vector(A=<1 1 1>, B=<-1.5 -1.5 -2>, Alpha=$System.FN_Integrator_v2.Attr)
    Velocity=&UniformRangedVector(Minimum=<-50 900 100>, Maximum=<50 1100 250>)
   ) {module=AddVelocity001}
  ]
  particleUpdateScript [
   &ParticleState()
   &GravityForce(Gravity=Linked() {linkedVariable {name=Particles.Gravity-InitialRange, type=Vector3f}})
   &Drag(
    Drag=&FloatFromCurve(FloatCurve=DataInterface() {propertyValues {Curve {keys [<at 0 7>,<at 1 1>]}}})
    Use Linear Drag=true
    Use Rotational Drag=false
   )
   &ScaleSpriteSize(Scale Factor=&Vector2DFromFloat(Value=&FloatFromCurve(
      FloatCurve=DataInterface() {propertyValues {Curve {keys [
          <at 0.01 0>
          {interpMode=RCIM_Cubic, tangentMode=RCTM_Break, time=0.02, value=0.75}
          {arriveTangent=0.002, interpMode=RCIM_Cubic, leaveTangent=0.002, tangentMode=RCTM_Break, time=0.05, value=1}
          {arriveTangent=-0.0496, interpMode=RCIM_Cubic, leaveTangent=-1.2885, tangentMode=RCTM_Break, time=0.7985, value=1}
          {interpMode=RCIM_Cubic, tangentMode=RCTM_Break, time=1.0011, value=0.002}
         ]}}}
      Scale Curve=&MakeCustomFloatFromBool(Bool=$User.bIsFullyDeployed, True Float=1, False Float=0)
     )))
   &SpriteRotationRate(Rotation Rate=&Lerp_Float(A=-1, B=1, Alpha=$Particles.Initial.MaterialRandom))
   &SpriteRotationRate(Rotation Rate=&Multiply_Float(
     A=&Lerp_Float(A=-800, B=800, Alpha=$Particles.Initial.MaterialRandom)
     B=&FloatFromCurve(FloatCurve=DataInterface() {propertyValues {Curve {keys [<at 0 8.9533>,<at 0.0642 1.9408>,<at 0.1376 0.9783>,<at 1 0.6849>]}}})
    )) {module=SpriteRotationRate001}
   &SolveForcesAndVelocity(Clamp Velocity=false, Limit Acceleration=false)
  ]
  rendererClasses [&NiagaraSpriteRendererProperties]
  renderers [{rendererClass=&NiagaraSpriteRendererProperties}]
 }
 Twigs001 {
  emitterUpdateScript [
   /SetVariables(Emitter.FullyDeployed:NiagaraFloat=&Lerp_Float(A=0, B=1, Alpha=$System.FN_Integrator_v2.Attr)) {isSetParameters=true, module=SetVariables_#1}
   &EmitterState(
    Life Cycle Mode=::NewEnumerator1
    Loop Duration=&UniformRangedFloat(Minimum=0.2, Maximum=2)
    Recalculate Duration Each Loop=true
    UseLoopDelay=false
   )
   &FN_PartitionAttributes_v2(Particles Per Partition:NiagaraInt32=&UniformRangedInt(Minimum=4, Maximum=15), Partitions=2)
   &SpawnRate(
    Spawn Probability=$Emitter.FullyDeployed
    SpawnRate=&Multiply_Float(
     A=&UniformRangedFloat(Minimum=0.2, Maximum=2)
     B=&FloatFromCurve(FloatCurve=DataInterface() {propertyValues {Curve {keys [<at 0.6 0>,<at 0.7 1>]}}}, CurveIndex=$Emitter.LoopedAge)
    )
    Use Spawn Probability=true
   )
  ]
  particleSpawnScript [
   &FN_ParticlePartition_v2()
   &InitializeParticle(
    Lifetime=&UniformRangedFloat(Minimum=2, Maximum=3)
    Mesh Scale=&Multiply_Vector(
     A=<0.8 0.8 0.8>
     B=&VectorFromFloat(Value=&UniformRangedFloat(Minimum=0.35, Maximum=0.9))
    )
    Module.Write Color=true
    Module.Write Lifetime=true
    Module.Write Mass=true
    Module.Write Material Random=true
    Module.Write Position=false
    Module.Write Scale=true
    Module.Write SpriteRotation=false
    Module.Write SpriteSize=false
    Position=&SimulationPosition()
    Sprite Rotation=&UniformRangedFloat(Minimum=0, Maximum=360)
    Sprite Size=&UniformRangedVector2D(Minimum=<6 9>, Maximum=<8 14>)
   )
   &FN_Initialize_BoneFlow_Attributes_v2(Angle=1, Radius=100)
   &FN_BoneFlow_v2(SkeletalMesh:NiagaraDataInterfaceSkeletalMesh=$User.SkeletalMesh, T:NiagaraFloat=$Particles.T)
   &AddVelocity(
    Scale Added Velocity=&Lerp_Vector(A=<1 1 1>, B=<-1.5 -1.5 -2>, Alpha=$System.FN_Integrator_v2.Attr)
    Velocity=&UniformRangedVector(Minimum=<-50 900 500>, Maximum=<50 1100 500>)
   ) {module=AddVelocity001}
   /Niagara/Modules/Spawn/Orientation/V2/InitialMeshOrientation(
    Rotation=&UniformRangedVector(Minimum=<-1 -1 -1>, Maximum=<1 1 1>)
    Use Orientation Vector=true
    Use Rotation Vector=true
   )
   /SetVariables(Particles.Gravity-InitialRange:Vector3f=&UniformRangedVector(Minimum=<0 0 -1000>, Maximum=<0 0 -1600>)) {isSetParameters=true}
  ]
  particleUpdateScript [
   &ParticleState()
   &GravityForce(Gravity=Linked() {linkedVariable {name=Particles.Gravity-InitialRange, type=Vector3f}})
   &Drag(
    Drag=&FloatFromCurve(FloatCurve=DataInterface() {propertyValues {Curve {keys [<at 0 7>,<at 1 1>]}}})
    Use Linear Drag=true
    Use Rotational Drag=false
   )
   &ScaleSpriteSize(Scale Factor=&Vector2DFromFloat(Value=&FloatFromCurve(
      FloatCurve=DataInterface() {propertyValues {Curve {keys [
          <at 0.019 0>
          {interpMode=RCIM_Cubic, tangentMode=RCTM_Break, time=0.02, value=0.75}
          {arriveTangent=0.002, interpMode=RCIM_Cubic, leaveTangent=0.002, tangentMode=RCTM_Break, time=0.05, value=1}
          {arriveTangent=-0.0496, interpMode=RCIM_Cubic, leaveTangent=-1.2885, tangentMode=RCTM_Break, time=0.7985, value=1}
          {interpMode=RCIM_Cubic, tangentMode=RCTM_Break, time=1.0011, value=0.002}
         ]}}}
      Scale Curve=&MakeCustomFloatFromBool(Bool=$User.bIsFullyDeployed, True Float=1, False Float=0)
     )))
   /Niagara/Modules/Update/Orientation/V2/UpdateMeshOrientation(
    Rotation Rate:NiagaraFloat=&Lerp_Float(A=-2, B=2, Alpha=$Particles.Initial.MaterialRandom)
    Rotation Vector:Vector3f=&UniformRangedVector(Minimum=<0 0 0>, Maximum=<1 1 1>)
   )
   &SolveForcesAndVelocity(Clamp Velocity=false, Limit Acceleration=false)
  ]
  rendererClasses [&NiagaraMeshRendererProperties]
  renderers [{rendererClass=&NiagaraMeshRendererProperties, rendererIndex=0}]
 }
}
summary {
 systemName=NS_Glider_Hightower_Grape_Trails
 emitters [
  {emitterName=FN_AnimTrail_V4001, rendererClasses [&NiagaraRibbonRendererProperties]}
  {emitterName=Leaves001, rendererClasses [&NiagaraSpriteRendererProperties]}
  {emitterName=Twigs001, rendererClasses [&NiagaraMeshRendererProperties]}
 ]
 userVariables [
  User.bIsFrontEndPreview {type=NiagaraBool}
  User.bIsFullyDeployed {type=NiagaraBool}
  User.ForwardDot {type=NiagaraFloat, defaultValue=0, description=""}
  User.LocalAlpha {type=NiagaraFloat, defaultValue=0, description=""}
  User.PlayerSpeed {type=NiagaraFloat, defaultValue=0, description=""}
  User.RibbonIDOffset {type=NiagaraInt32}
  User.RightDot {type=NiagaraFloat, defaultValue=0, description=""}
  User.SkeletalMesh {
   type=NiagaraDataInterfaceSkeletalMesh
   defaultValue {
    struct=NiagaraExt_VariableValue_DataInterface
    value {dataInterface=&self:NiagaraDataInterfaceSkeletalMesh_0}
   }
  }
 ]
}
